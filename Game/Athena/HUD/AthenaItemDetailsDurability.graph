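# Item details HUD widget that shows an item's durability as a percentage and bar, compared against the equipped item.
# When the represented item changes, it feeds the world item to a durability viewmodel and collapses if the item has no durability.
# DetectAnyDurabilityChanged compares both viewmodels to pick the arrow, equal or normal icon and text color.
# DurabilityToColorA/B map durability to three-tier color pairs using low and very-low thresholds.
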WidgetBlueprint /Game/Athena/HUD/AthenaItemDetailsDurability
graphs [
 DetectAnyDurabilityChanged {nodes [
   DetectAnyDurabilityChanged!() -> (then, NewDurability:double=~) {is=title K2Node_FunctionEntry}
   n1 = Get InteractingItemDurabilityVM() -> (InteractingItemDurabilityVM:Item Durability Viewmodel Object Reference)
   > Is Valid(InputObject=@n1)
   n2 = GetItemDurability!(self:Item Durability Viewmodel Object Reference=@n1, bCheckInitialized:bool=true) -> (ReturnValue:float)
   n3 = Get FortEquippedItemDurabilityVM() -> (FortEquippedItemDurabilityVM:Fort Equipped Item Durability VM Object Reference)
   n4 = GetItemDurability(self=@n3, bCheckInitialized=false)
   n5 = ItemIsComparableToOther!(ItemDurabilityVMLHS:Item Durability Viewmodel Object Reference=@n1, ItemDurabilityVMRHS:Item Durability Viewmodel Object Reference=@n3) -> (ReturnValue:bool)
   > SetComparisonImage!(execute, self, AreComparable:bool=@n5, InteractingDurability:double=@n2, EquippedDurability:double=@n4) -> (then)
   > SetPercentTextColor!(execute, self, AreComparable:bool=@n5, InteractingDurability:double=@n2, EquippedDurability:double=@n4) -> (then)
  ]}
 DurabilityToColorA {nodes [
   n1 = DurabilityToColorA!() -> (then, InDurability:double=~) {is=title K2Node_FunctionEntry}
   > n2 = Branch(Condition="float > float"(A=@n1, B=Get LowDurabilityThreshold_0() -> (LowDurabilityThreshold_0:double)))
   > Return Node!(execute, OutColor:Linear Color Structure=(R=0,G=0.3185,B=0.0331,A=1))
   n3 = Branch(execute=@n2.else, Condition="float > float"(A=@n1, B=Get VeryLowDurabilityThreshold() -> (VeryLowDurabilityThreshold:double)))
   > Return Node(OutColor=(R=1,G=0.227,B=0.0331,A=1))
   Return Node(execute=@n3.else, OutColor=(R=1,G=0.0331,B=0.0331,A=1))
  ]}
 DurabilityToColorB {nodes [
   n1 = DurabilityToColorB!() -> (then, InDurability:double=~) {is=title K2Node_FunctionEntry}
   > n2 = Branch(Condition="float > float"(A=@n1, B=Get LowDurabilityThreshold_0()))
   > Return Node(OutColor=(R=0.1329,G=0.6038,B=0.0331,A=1))
   n3 = Branch(execute=@n2.else, Condition="float > float"(A=@n1, B=Get VeryLowDurabilityThreshold()))
   > Return Node(OutColor=(R=1,G=0.6038,B=0,A=1))
   Return Node(execute=@n3.else, OutColor=(R=1,G=0.1248,B=0.0331,A=1))
  ]}
 EventGraph {nodes [
   n1 = Event OnPostItemToRepresentChanged!() -> (OutputDelegate, then, ItemToRepresent:Fort Item Object Reference) {is=title K2Node_Event}
   > n2 = Cast To FortWorldItem!(execute, Object=@n1.ItemToRepresent) -> (then, CastFailed, AsFort World Item:Fort World Item Object Reference) {is=title K2Node_DynamicCast}
   n3 = Get InteractingItemDurabilityVM()
   > SetWorldItem!(execute, self:Item Durability Viewmodel Object Reference=@n3, InWorldItem:Fort World Item Object Reference=@n2, bCheckInitialized:bool=true) -> (then)
   > SetVisibility(InVisibility=Select(
     Option 0=Collapsed
     Option 1=HitTestInvisible
     Index=ItemHasDurability!(self:Item Durability Viewmodel Object Reference=@n3) -> (ReturnValue:bool)
    ))
  ]}
 SetComparisonImage {nodes [
   SetComparisonImage!() -> (then, AreComparable:bool=~, InteractingDurability:double=~, EquippedDurability:double=~) {is=title K2Node_FunctionEntry}
   n1 = SelectMaterialInstance!(
    A:Material Instance Object Reference=/Game/Athena/HUD/Art/DurabilityIcon/MI_UI_DurabilityComparison_Icon_ArrowUp
    B:Material Instance Object Reference=/Game/Athena/HUD/Art/DurabilityIcon/MI_UI_DurabilityComparison_Icon_ArrowDown
    bSelectA:bool="float > float"(A=Get InteractingDurability() -> (InteractingDurability:double), B=Get EquippedDurability() -> (EquippedDurability:double))
   ) -> (ReturnValue:Material Instance Object Reference)
   n2 = "Equal (Float)"(A=Get InteractingDurability(), B=Get EquippedDurability())
   > SetBrushFromMaterial(
    self=Get ComparisonImage() -> (ComparisonImage:Image Object Reference)
    Material=SelectMaterialInstance(
     A=SelectMaterialInstance(A=/Game/Athena/HUD/Art/DurabilityIcon/MI_UI_DurabilityComparison_Icon_Equal, B=@n1, bSelectA=@n2)
     B=/Game/Athena/HUD/Art/DurabilityIcon/MI_UI_DurabilityComparison_Icon_Normal
     bSelectA=Get AreComparable() -> (AreComparable:bool)
    )
   )
  ]}
 SetPercentTextColor {nodes [
   SetPercentTextColor!() -> (then, AreComparable:bool=~, InteractingDurability:double=~, EquippedDurability:double=~) {is=title K2Node_FunctionEntry}
   n1 = SelectColor(
    A=(R=0.4342,G=1,B=0.4342,A=1)
    B=(R=1,G=0.1714,B=0.1714,A=1)
    bPickA="float > float"(A=Get InteractingDurability(), B=Get EquippedDurability())
   )
   n2 = "Equal (Float)"(A=Get InteractingDurability(), B=Get EquippedDurability())
   > SetColorAndOpacity!(
    execute
    self=Get Percentage() -> (Percentage:Common Text Object Reference)
    InColorAndOpacity_SpecifiedColor:Linear Color Structure=SelectColor(A=SelectColor(A=(R=1,G=1,B=0.4508,A=1), B=@n1, bPickA=@n2), B=(R=1,G=1,B=1,A=1), bPickA=Get AreComparable())
    InColorAndOpacity_ColorUseRule:ESlateColorStylingMode Enum
   ) -> (then)
  ]}
]
variables [
 LowDurabilityThreshold {type=number}
 LowDurabilityThreshold_0 {type=number, default=0.5}
 VeryLowDurabilityThreshold {type=number, default=0.25}
]
widgets {
 info {inheritedWidgetCount=0, parentClass=/Script/FortniteUI.AthenaItemElementWidgetBase, rootWidgetClass=&Overlay, widgetCount=10}
 widgets [
  Overlay_20 {widgetClassPath=&Overlay, widget=Overlay_20, properties {visibility=HitTestInvisible}}
  GridPanel_0 {widgetClassPath=&GridPanel, parent=Overlay_20, widget=GridPanel_0, slot {class=OverlaySlot}}
  DurabilityComparisonView {widgetClassPath=&HorizontalBox, parent=GridPanel_0, widget=DurabilityComparisonView, slot {class=GridSlot}}
  SizeBox_0 {
   widgetClassPath=&SizeBox
   parent=DurabilityComparisonView
   widget=SizeBox_0
   properties {bOverride_HeightOverride=true, bOverride_WidthOverride=true, heightOverride=32, widthOverride=32}
   slot {class=HorizontalBoxSlot, properties {verticalAlignment=VAlign_Center, padding=<ltrb 0 0 8 0>}}
  }
  ComparisonImage {
   widgetClassPath=&Image
   bIsVariable=true
   parent=SizeBox_0
   widget=ComparisonImage
   properties {brush {
     bIsDynamicallyLoaded=false
     drawAs=Image
     imageType=FullColor
     mirroring=NoMirror
     resourceName=None
     resourceObject=&MI_UI_DurabilityComparison_Icon_Normal
     tiling=NoTile
     imageSize=<48 48>
     margin=<ltrb 0 0 0 0>
     outlineSettings {
      bUseBrushTransparency=false
      roundingType=HalfHeightRadius
      width=0
      color {colorUseRule=UseColor_Specified, specifiedColor=<rgba 0 0 0 0>}
      cornerRadii=<0 0 0 0>
     }
     tintColor {colorUseRule=UseColor_Specified, specifiedColor=<rgba 1 1 1 1>}
     uVRegion {bIsValid=false, max=<0 0>, min=<0 0>}
    }}
   slot {class=SizeBoxSlot, properties {horizontalAlignment=HAlign_Center, verticalAlignment=VAlign_Center}}
  }
  Percentage {
   widgetClassPath=&CommonTextBlock
   bIsVariable=true
   parent=DurabilityComparisonView
   widget=Percentage
   properties {
    autoWrapText=true
    justification=Right
    text=20
    textTransformPolicy=ToUpper
    font {
     bForceMonospaced=false
     bMaterialIsStencil=false
     fontMaterial=null
     fontObject=&Burbank
     letterSpacing=0
     monospacedWidth=1
     size=19.5
     skewAmount=0
     typefaceFontName=BurbankRegular-Black
     outlineSettings {
      bApplyOutlineToDropShadows=true
      bMiteredCorners=false
      bSeparateFillAlpha=true
      outlineMaterial=&M-FontShadow
      outlineSize=2
      outlineColor=<rgba 0 0 0 0.5>
     }
    }
    shadowOffset=<0 0>
    shapedTextOptions {bOverride_TextFlowDirection=true, bOverride_TextShapingMethod=false, textFlowDirection=LeftToRight, textShapingMethod=Auto}
   }
   slot {class=HorizontalBoxSlot, properties {horizontalAlignment=HAlign_Right, verticalAlignment=VAlign_Center, padding=<ltrb 0 0 8 0>}}
  }
  SizeBox_1 {
   widgetClassPath=&SizeBox
   parent=DurabilityComparisonView
   widget=SizeBox_1
   properties {bOverride_HeightOverride=true, bOverride_WidthOverride=true, heightOverride=12, widthOverride=180}
   slot {class=HorizontalBoxSlot, properties {verticalAlignment=VAlign_Center}}
  }
  Overlay_512 {widgetClassPath=&Overlay, parent=SizeBox_1, widget=Overlay_512, slot {class=SizeBoxSlot}}
  Outline {
   widgetClassPath=&Image
   parent=Overlay_512
   widget=Outline
   properties {brush {
     bIsDynamicallyLoaded=false
     drawAs=Image
     imageType=NoImage
     mirroring=NoMirror
     resourceName=None
     resourceObject=null
     tiling=NoTile
     imageSize=<32 32>
     margin=<ltrb 0 0 0 0>
     outlineSettings {
      bUseBrushTransparency=false
      roundingType=HalfHeightRadius
      width=0
      color {colorUseRule=UseColor_Specified, specifiedColor=<rgba 0 0 0 0>}
      cornerRadii=<0 0 0 0>
     }
     tintColor {colorUseRule=UseColor_Specified, specifiedColor=<rgba 1 1 1 0.7>}
     uVRegion {bIsValid=false, max=<0 0>, min=<0 0>}
    }}
   slot {class=OverlaySlot, properties {horizontalAlignment=HAlign_Fill, verticalAlignment=VAlign_Fill, padding=<ltrb -1 -1 -1 -1>}}
  }
  SimpleMaterialProgressBar {
   widgetClassPath=&/SimpleMaterialProgressBar
   bIsVariable=true
   parent=Overlay_512
   widget=SimpleMaterialProgressBar
   properties {
    bOverrideDefaultColorA=true
    bOverrideDefaultColorB=true
    bOverrideDefaultColorBackground=true
    designTime_Progress=0.7011
    cachedColorA=<rgba 0.0742 0.5029 0.0356 1>
    cachedColorB=<rgba 0.1779 0.7305 0.0782 1>
    cachedColorBackground=<rgba 0 0 0 0.8>
   }
   slot {class=OverlaySlot, properties {horizontalAlignment=HAlign_Fill, verticalAlignment=VAlign_Fill}}
  }
 ]
}
