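# Item details header widget, a single text block that shows the detailed item's name in uppercase.
# Setting a different detail or compare item refreshes the text from ItemToDetail and collapses it when the name is empty.
# If the item's schematic crafting result is a weapon, the text takes Color4 of the item's rarity data.
# Otherwise, a valid item has the text's current style re-applied.
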WidgetBlueprint /Game/UI/Foundation/Items/Details/ItemDetailsHeaderItemDisplayText
graphs [
 EventGraph {nodes [
   n1 = Event HandlePostDifferentItemToDetailSet() {is=title K2Node_Event}
   n2 = GetDisplayName!() -> (DisplayName:Text)
   "SetText (Text)"(
    execute=[@n1.then, @n6.then]
    self=Get ItemDisplayName() -> (ItemDisplayName:Common Text Object Reference)
    InText=TextToUpper!(InText:"Text (by ref)"=@n2) -> (ReturnValue:Text)
   )
   > SetVisibility(self=Get ItemDisplayName(), InVisibility=Select(Option 0=Visible, Option 1=Collapsed, Index=TextIsEmpty(InText=@n2)))
   > Is Valid(InputObject=Get ItemToDetail() -> (ItemToDetail:Fort Item Weak object))
   > n3 = Cast To FortWeaponItemDefinition!(
    execute
    Object=GetSchematicCraftingResult!(self:Fort Item Object Reference=Get ItemToDetail()) -> (ReturnValue:Fort Item Definition Object Reference)
   ) -> (then, CastFailed, AsFort Weapon Item Definition:Fort Weapon Item Definition Object Reference) {is=title K2Node_DynamicCast}
   > n4 = GetRarityDataForItemDef!(
    execute
    ItemDef:Item Definition Base Object Reference=GetItemDefinition!(self:Fort Item Object Reference=Get ItemToDetail()) -> (ReturnValue:Fort Item Definition Object Reference)
   ) -> (then, ReturnValue:Fort Rarity Item Data Structure)
   > SetColorAndOpacity(
    self=Get ItemDisplayName()
    InColorAndOpacity=Make Slate Color(SpecifiedColor=Break Fort Rarity Item Data!(FortRarityItemData:"Fort Rarity Item Data Structure (by ref)"=@n4) -> (
      Name:Text
      Color1:Linear Color Structure
      Color2:Linear Color Structure
      Color3:Linear Color Structure
      Color4:Linear Color Structure
      Color5:Linear Color Structure
      Radius:float=~
      Falloff:float=~
      Brightness:float=~
      Roughness:float=~
      Glow:float=~
     ) {is=title K2Node_BreakStruct}.Color4)
   )
   n5 = Get ItemDisplayName()
   SetStyle!(
    execute=@n3.CastFailed
    self:Common Text Object Reference=@n5
    InStyle:Common Text Style Class Reference=Get Style!(self:Common Text Object Reference=@n5) -> (Style:Common Text Style Class Reference)
   ) -> (then)
   n6 = Event HandlePostDifferentItemToCompareWithSet() {is=title K2Node_Event}
  ]}
 GetDisplayName {nodes [
   GetDisplayName() {is=title K2Node_FunctionEntry}
   > n1 = Is Valid(InputObject=Get ItemToDetail())
   Return Node!(execute=@n1.Is Not Valid, DisplayName:Text)
   > Return Node(DisplayName=GetDisplayName!(self:Fort Item Object Reference=Get ItemToDetail()) -> (ReturnValue:Text) {is=title K2Node_CallFunction})
  ]}
]
variables []
widgets {
 info {
  inheritedWidgetCount=0
  parentClass=/Script/FortniteUI.FortItemDetailElementWidget
  rootWidgetClass=&CommonTextBlock
  widgetCount=1
 }
 widgets [ItemDisplayName {
   widgetClassPath=&CommonTextBlock
   bIsVariable=true
   widget=ItemDisplayName
   properties {
    scrollStyle=&TextScrollStyle-Base_C
    style=/FortUILibrary/Styles/Typography/Foundation/TextStyle-Header-S.TextStyle-Header-S_C
    text="{ItemName}"
    font {
     bForceMonospaced=false
     bMaterialIsStencil=false
     fontMaterial=null
     fontObject=&Burbank
     letterSpacing=0
     monospacedWidth=1
     size=24
     skewAmount=0
     typefaceFontName=Burbank-Black
     outlineSettings {
      bApplyOutlineToDropShadows=false
      bMiteredCorners=false
      bSeparateFillAlpha=false
      outlineMaterial=null
      outlineSize=0
      outlineColor=<rgba 0 0 0 1>
     }
    }
    shadowOffset=<0 0>
   }
  }]
}
