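# Item detail widget, a FortItemDetailElementWidget, that shows the detailed item's quantity as an "x" plus number badge.
# Item, quantity-override and pre-construct events refresh the count, with OnItemChanged rebound whenever the item switches.
# Count prefers a non-negative QuantityOverride, then a world item's positive fuel charge, then stack size, or zero without an item.
# Text is optionally truncated, and the badge hides unless Count exceeds one or AlwaysVisible is set.
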WidgetBlueprint /Game/UI/Foundation/Items/ItemDetailsStackCounter
graphs [
 EventGraph {nodes [
   n1 = HandleOnItemChanged!() -> (OutputDelegate, then, bItemChanged:bool=~, bAmmoChanged:bool=~, bIngredientsChanged:bool=~) {is=title K2Node_CustomEvent}
   > Update Stack Count()
   Event HandleQuantityOverrideChanged() {is=title K2Node_Event}
   > Update Stack Count()
   Event HandlePreDifferentItemToDetailSet() {is=title K2Node_Event}
   n2 = Get ItemToDetail() -> (ItemToDetail:Fort Item Weak object)
   > Is Valid(InputObject=@n2)
   > Unbind Event from On Item Changed(self:Fort Item Object Reference=@n2, Delegate=@n1) {is=title K2Node_RemoveDelegate}
   Event HandlePostDifferentItemToDetailSet() {is=title K2Node_Event}
   > Update Stack Count()
   > Is Valid(InputObject=@n2)
   > Bind Event to On Item Changed(self:Fort Item Object Reference=@n2, Delegate=@n1) {is=title K2Node_AddDelegate}
   Event PreConstruct()
   > Update Stack Count()
   Event Construct()
  ]}
 Refresh Visibility {nodes [
   Refresh Visibility() {is=title K2Node_FunctionEntry}
   n1 = "integer > integer"(A=Get Count() -> (Count:int=0), B=1) {is=title K2Node_CallFunction}
   n2 = Get AlwaysVisible() -> (AlwaysVisible:bool)
   > SetVisibility(self=Self-Reference(), InVisibility=Select(Option 0=Hidden, Option 1=HitTestInvisible, Index=OR Boolean(A=@n1, B=@n2)))
  ]}
 Update Stack Count {nodes [
   Update Stack Count() {is=title K2Node_FunctionEntry}
   n1 = Get ItemToDetail()
   > n2 = Is Valid(InputObject=@n1)
   > n3 = Branch(Condition="integer >= integer"!(A:int=Get QuantityOverride() -> (QuantityOverride:int=0), B:int=0) -> (ReturnValue:bool) {is=title K2Node_CallFunction})
   > n4 = Set Count(Count:int=Get QuantityOverride() -> (QuantityOverride=0)) -> (Output_Get:int=0)
   n5 = Get Count() -> (Count=0)
   "SetText (Text)"(
    execute=[@n9.then, @n4.then, @n11.then, @n10.then]
    self=Get TextCounter() -> (TextCounter:Common Text Object Reference)
    InText=Select(
     Option 0:Text="To Text (Integer)"(Value=@n5)
     Option 1:Text=Truncate Integer Value!(Value:int=@n5, Min Fractional Digits:int=0, Max Fractional Digits:int=1) -> (Formatted Value:Text)
     Index=Get TruncateValue() -> (TruncateValue:bool)
    ) -> (ReturnValue:Text)
   )
   > Refresh Visibility() {is=title K2Node_CallFunction}
   n6 = Cast To FortWorldItem!(execute=@n3.else, Object=@n1) -> (then, CastFailed, AsFort World Item:Fort World Item Object Reference) {is=title K2Node_DynamicCast}
   n7 = GetFuelChargeValue!(self:Fort World Item Object Reference=@n6) -> (ReturnValue:int=0)
   > n8 = Branch(Condition="integer > integer"(A=@n7, B=0))
   > n9 = Set Count(Count=@n7) -> (Output_Get=0)
   n10 = Set Count(execute=[@n8.else, @n6.CastFailed], Count=GetNumInStack!(self:Fort Item Object Reference=Get ItemToDetail()) -> (ReturnValue:int=0)) -> (Output_Get=0)
   n11 = Set Count(execute=@n2.Is Not Valid, Count=0) -> (Output_Get=0)
  ]}
]
variables [
 AlwaysVisible {type=boolean}
 Count {type=integer, default=-1}
 TruncateValue {type=boolean, default=true}
]
widgets {
 info {inheritedWidgetCount=0, parentClass=/Script/FortniteUI.FortItemDetailElementWidget, rootWidgetClass=&SizeBox, widgetCount=5}
 widgets [
  "[SizeBox]" {
   widgetClassPath=&SizeBox
   bIsVariable=true
   widget="[SizeBox]"
   properties {bOverride_HeightOverride=true, bOverride_MinDesiredWidth=true, heightOverride=26, minDesiredWidth=54, widthOverride=54}
  }
  BorderBase {
   widgetClassPath=&CommonBorder
   bIsVariable=true
   parent="[SizeBox]"
   widget=BorderBase
   properties {
    horizontalAlignment=HAlign_Center
    style=/FortUILibrary/Styles/Border/Border-Bang.Border-Bang_C
    verticalAlignment=VAlign_Center
    background {
     bIsDynamicallyLoaded=false
     drawAs=Box
     imageType=NoImage
     mirroring=NoMirror
     resourceName=None
     resourceObject=/FortUILibrary/Styles/Border/Materials/M-Border-MainL-Black
     tiling=NoTile
     imageSize=<128 32>
     margin=<ltrb 0.1 0.45 0.9 0.5>
     outlineSettings {
      bUseBrushTransparency=false
      roundingType=HalfHeightRadius
      width=0
      color {colorUseRule=UseColor_Specified, specifiedColor=<rgba 0 0 0 0>}
      cornerRadii=<0 0 0 0>
     }
     tintColor {colorUseRule=UseColor_Specified, specifiedColor=<rgba 0 0 0 1>}
     uVRegion {bIsValid=false, max=<0 0>, min=<0 0>}
    }
    padding=<ltrb 15 0 15 0>
   }
   slot {class=SizeBoxSlot}
  }
  HorizontalBox_0 {
   widgetClassPath=&HorizontalBox
   parent=BorderBase
   widget=HorizontalBox_0
   properties {renderTransform {angle=0, scale=<1 1>, shear=<0 0>, translation=<0 -2>}}
   slot {class=BorderSlot, properties {horizontalAlignment=HAlign_Center, verticalAlignment=VAlign_Center, padding=<ltrb 15 0 15 0>}}
  }
  CommonTextBlock_1 {
   widgetClassPath=&CommonTextBlock
   parent=HorizontalBox_0
   widget=CommonTextBlock_1
   properties {
    style=/FortUILibrary/Styles/Typography/Foundation/TextStyle-Base-XS-Black.TextStyle-Base-XS-Black_C
    text=x
    colorAndOpacity {colorUseRule=UseColor_Specified, specifiedColor=<rgba 0 0 0 1>}
    font {
     bForceMonospaced=false
     bMaterialIsStencil=false
     fontMaterial=null
     fontObject=&NotoSans
     letterSpacing=0
     monospacedWidth=1
     size=16
     skewAmount=0
     typefaceFontName=NotoSans-Regular
     outlineSettings {
      bApplyOutlineToDropShadows=false
      bMiteredCorners=false
      bSeparateFillAlpha=false
      outlineMaterial=null
      outlineSize=0
      outlineColor=<rgba 0 0 0 1>
     }
    }
    shadowOffset=<0 0>
   }
   slot {class=HorizontalBoxSlot, properties {horizontalAlignment=HAlign_Center, verticalAlignment=VAlign_Center}}
  }
  TextCounter {
   widgetClassPath=&CommonTextBlock
   bIsVariable=true
   parent=HorizontalBox_0
   widget=TextCounter
   properties {
    style=/FortUILibrary/Styles/Typography/Foundation/TextStyle-Base-XS-B-Black.TextStyle-Base-XS-B-Black_C
    text=Text Block
    colorAndOpacity {colorUseRule=UseColor_Specified, specifiedColor=<rgba 0 0 0 1>}
    font {
     bForceMonospaced=false
     bMaterialIsStencil=false
     fontMaterial=null
     fontObject=&NotoSans
     letterSpacing=0
     monospacedWidth=1
     size=16
     skewAmount=0
     typefaceFontName=NotoSans-Bold
     outlineSettings {
      bApplyOutlineToDropShadows=false
      bMiteredCorners=false
      bSeparateFillAlpha=false
      outlineMaterial=null
      outlineSize=0
      outlineColor=<rgba 0 0 0 1>
     }
    }
    shadowOffset=<0 0>
   }
   slot {class=HorizontalBoxSlot}
  }
 ]
}
